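# Conversation database for the Bright Ion NPC, a Supply specialist, defining its greeting dialogue.
# The entry point leads to one speech node with a default line plus contextual lines gated by first-ever-conversation and per-match conversation-count requirements.
# The first-meeting line takes its priority from a speech priority curve row.
# The speech node then links out to the shared NPC Library quests and basic services conversations.
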ConversationDatabase /NPCLibrary/NPCs/BrightIon/NPCLibrary_Conversation_BrightIon
entry_points [{tag=Conversation.Entry.BrightIon, destinations [#1]}]
nodes [
 ConversationEntryPointNode_7 {
  id=#1
  class=ConversationEntryPointNode
  outputs [#2]
  properties {entryTag {tagName=Conversation.Entry.BrightIon}}
 }
 Speech_C_4 {
  id=#2
  class=Speech_C
  outputs [#3, #4]
  properties {
   nodeName=""
   generalConfig {
    defaultMessage="Heyo!"
    contextualMessages [
     {
      message="I don't know about you, but I love it here."
      contextRequirements [{requirement {class=FortControllerRequirement_IsFirstEverConversationWithNPC}}]
      priority {curve {rowName=Tandem.Speech.Priority.FirstEverInteraction}}
     }
     {
      message="Heyo!"
      contextRequirements [{requirement {
         class=FortControllerRequirement_CheckNPCMatchConversationCount
         properties {nPCUniqueIdTag {tagName=AISpawnerData.Type.Specialist.Supply.BrightIon}}
        }}]
      priority {curve {rowName=None}}
     }
     {
      message="Just passin' through!"
      contextRequirements [{requirement {
         class=FortControllerRequirement_CheckNPCMatchConversationCount
         properties {nPCUniqueIdTag {tagName=AISpawnerData.Type.Specialist.Supply.BrightIon}}
        }}]
      priority {curve {rowName=None}}
     }
    ]
   }
  }
 }
 ConversationLinkNode_0 {id=#4, class=ConversationLinkNode, properties {remoteEntryTag {tagName=Conversation.Entry.BasicServices}}}
 ConversationLinkNode_1 {id=#3, class=ConversationLinkNode, properties {remoteEntryTag {tagName=Conversation.Entry.NPCLibrary.Quests}}}
]
